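# Behavior tree for specialist NPCs, judging by its path an injected subtree, that delivers supplies to the ActorToSupply blackboard target.
# A sequence moves the bot near that actor, turns to face it, waits briefly, then executes the GiveSupplies gameplay ability by tag to drop the pickup.
# Throughout the sequence, services apply the LeashPaused status tag and run the melee attack evaluator.
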BehaviorTree /NPCSpecialists/BTInjection/BT_GiveSupplies
blackboard=&/BB_GiveSupplies
root {
 name=Sequence
 class=BTComposite_Sequence
 children [Sequence {
   class=BTComposite_Sequence
   children [
    Bot Move To {
     class=FortAthenaBTTask_BotMoveTo
     decorators []
     properties {
      movementResultKeyName=None
      nodeName=Bot Move To
      acceptableRadius {defaultValue=340}
      allowRandomWobble {defaultValue=true}
      bAllowStrafe {defaultValue=false}
      blackboardKey {selectedKeyName=ActorToSupply}
      observedBlackboardValueTolerance {defaultValue=4.75}
     }
    }
    Rotate to face BB entry {class=BTTask_RotateToFaceBBEntry, decorators [], properties {blackboardKey {selectedKeyName=ActorToSupply}, precision {defaultValue=10}}}
    Wait {class=BTTask_Wait, properties {waitTime {defaultValue=1}}}
    "Drop Pickup " {
     class=FortBTTask_ExecuteGameplayAbility
     decorators []
     properties {nodeName="Drop Pickup ", gameplayAbilityTagContainer {defaultValue {gameplayTags [{tagName=Ability.Specialist.Supply.GiveSupplies}]}}}
    }
   ]
   decorators []
   properties {
    children [
     {childTask=&self:FortAthenaBTTask_BotMoveTo_0}
     {childTask=&self:BTTask_RotateToFaceBBEntry_0}
     {childTask=&self:BTTask_Wait_13}
     {childTask=&self:FortBTTask_ExecuteGameplayAbility_2}
    ]
    services [
     &self:FortAthenaBTService_ApplyGameplayTags_1
     &self:FortAthenaBTService_AIEvaluator_10
    ]
   }
   services [
    Pause Leash {
     class=FortAthenaBTService_ApplyGameplayTags
     properties {interval=0.5, nodeName=Pause Leash, gameplayTagsToApply {defaultValue {gameplayTags [{tagName=Athena.Gameplay.Status.LeashPaused}]}}}
    }
    Evaluating...Melee Attack {
     class=FortAthenaBTService_AIEvaluator
     properties {
      interval=0.3
      nodeName=Evaluating...Melee Attack
      aIEvaluatorClass {defaultValue=&FortAthenaAIBotEvaluator_MeleeAttack}
     }
    }
   ]
  }]
 decorators []
 properties {children [{childComposite=&self:BTComposite_Sequence_0, childTask=null}]}
}
