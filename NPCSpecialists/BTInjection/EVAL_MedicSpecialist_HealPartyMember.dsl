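; Behavior tree injection evaluator that lets a hired medic specialist NPC pick a party member to heal.
; On enter it caches the pawn's converter, player state and game mode.
; The start check scans the converter's shuffled squad, or only a bot converter, for the first nearby, non-downed member missing health or shield, despite the "closest" variable name.
; When the action is injected, that member is written to the supply-target blackboard key.
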(event EventOnEnter
  (params (OwnerComp :type Behavior%20Tree%20Component%20Object%20Reference))
  (bind _ascontroller (Utilities|Casting|CastToController (Components|GetOwner OwnerComp)))
  (bind _output_get (Variables|Default|SetAIController _ascontroller))
  (Variables|Default|SetPawn (Pawn|GetControlledPawn _output_get))
  (bind _output_get_2 (Variables|Default|SetConverterPawn (Convert|GetConverterPawn (Actor|GetComponentbyClass (Variables|Default|GetPawn) "/Script/FortniteGame.FortPawnComponent_Convert"))))
  (bind _asfort_player_state_athena (Utilities|Casting|CastToFortPlayerStateAthena (Variables|/Script/Engine.Controller|GetPlayerState (Pawn|GetController _output_get_2))))
  (Variables|Default|SetConverterPlayerStateAthena _asfort_player_state_athena)
  (bind _asfort_game_mode_athena (Utilities|Casting|CastToFortGameModeAthena (Game|GetGameMode)))
  (Variables|Default|SetGameModeAthena _asfort_game_mode_athena))

(event EventOnCorrespondingActionInjected
  (params (OwnerComp :type Behavior%20Tree%20Component%20Object%20Reference))
  (AI|Components|Blackboard|SetValueasObject (AI|GetBlackboard (Variables|Default|GetAIController)) (Variables|Default|GetBBKeyActorToSupplyName) (Variables|Default|GetClosestTeamMemberFound)))

(fn EvaluateStartingConditions
  (params (OwnerComp :type Behavior%20Tree%20Component%20Object%20Reference))
  (|CallParentFunction OwnerComp)
  (if (Pawn|IsBotControlled (Variables|Default|GetConverterPawn))
    (bind _asfort_player_state_athena (Utilities|Casting|CastToFortPlayerStateAthena (Variables|/Script/Engine.Pawn|GetPlayerState (Variables|Default|GetConverterPawn))))
    (Utilities|Array|AddUnique (LocalVariables|GetFoundSquadMembers) _asfort_player_state_athena)
    (else
      (bind _squadmembers (Class|FortGameModeAthena|GetSquadMembers (Variables|Default|GetGameModeAthena) (Class|FortPlayerStateAthena|GetSquadId (Variables|Default|GetConverterPlayerStateAthena))))
      (bind _output_get (LocalVariables|SetFoundSquadMembers _squadmembers))
      (Utilities|Array|Shuffle _output_get)))
  (for _array_element (LocalVariables|GetFoundSquadMembers)
    (if (Math|Boolean|ANDBoolean (<= (Math|Vector|Distance2D(Vector) (Transformation|GetActorLocation (PlayerState|GetPawn _array_element)) (Transformation|GetActorLocation (Variables|Default|GetPawn))) (ScalableFloat|GetValueatLevel (Variables|Default|GetDistanceToTeamMember))) (not (Variables|/Script/FortniteGame.FortPawn|GetbIsDBNO (Utilities|Casting|CastToFortPlayerPawn (PlayerState|GetPawn _array_element)))) (or (< (Math|Float|SafeDivide (FortPawn|GetHealth (Utilities|Casting|CastToFortPlayerPawn (PlayerState|GetPawn _array_element))) (FortPawn|GetMaxHealth (Utilities|Casting|CastToFortPlayerPawn (PlayerState|GetPawn _array_element)))) 1.0) (< (Math|Float|SafeDivide (FortPawn|GetShield (Utilities|Casting|CastToFortPlayerPawn (PlayerState|GetPawn _array_element))) (FortPawn|GetMaxShield (Utilities|Casting|CastToFortPlayerPawn (PlayerState|GetPawn _array_element)))) 1.0)))
      (bind _asfort_player_pawn (Utilities|Casting|CastToFortPlayerPawn (PlayerState|GetPawn _array_element)))
      (Variables|Default|SetClosestTeamMemberFound _asfort_player_pawn)
      (return true)))
  (Variables|Default|SetClosestTeamMemberFound 0)
  (return false))
