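; Behavior tree injection evaluator that decides when a hired supply specialist NPC can give supplies to a squad member.
; On enter it caches the AI controller, pawn, game mode and the player state of the converting player.
; The start check shuffles that player's squad and takes the first member within a scalable 2D distance, so the pick is random despite the "closest" variable name.
; On injection it writes that member to the blackboard as the actor to supply.
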(event EventOnEnter
  (params (OwnerComp :type Behavior%20Tree%20Component%20Object%20Reference))
  (bind _pawn (Variables|Default|GetPawn))
  (bind _returnvalue (Actor|GetComponentbyClass _pawn "/Script/FortniteGame.FortPawnComponent_Convert"))
  (bind _returnvalue_2 (Convert|GetConverterPawn _returnvalue))
  (bind _ascontroller (Utilities|Casting|CastToController (Components|GetOwner OwnerComp)))
  (bind _output_get (Variables|Default|SetAIController _ascontroller))
  (Variables|Default|SetPawn (Pawn|GetControlledPawn _output_get))
  (bind _asfort_game_mode_athena (Utilities|Casting|CastToFortGameModeAthena (Game|GetGameMode)))
  (Variables|Default|SetGameModeAthena _asfort_game_mode_athena)
  (Utilities|IsValid _returnvalue_2
    (:Is%20Valid
      (bind _output_get_2 (Variables|Default|SetConverterPawn _returnvalue_2))
      (bind _asfort_player_state_athena (Utilities|Casting|CastToFortPlayerStateAthena (Variables|/Script/Engine.Controller|GetPlayerState (Pawn|GetController _output_get_2))))
      (Variables|Default|SetConverterPlayerStateAthena _asfort_player_state_athena))))

(event EventOnCorrespondingActionInjected
  (params (OwnerComp :type Behavior%20Tree%20Component%20Object%20Reference))
  (AI|Components|Blackboard|SetValueasObject (AI|GetBlackboard (Variables|Default|GetAIController)) (Variables|Default|GetBBKeyActorToSupplyName) (Variables|Default|GetClosestTeamMemberFound)))

(fn EvaluateStartingConditions
  (params (OwnerComp :type Behavior%20Tree%20Component%20Object%20Reference))
  (bind _converterplayerstateathena (Variables|Default|GetConverterPlayerStateAthena))
  (|CallParentFunction OwnerComp)
  (Utilities|IsValid _converterplayerstateathena
    (:Is%20Valid
      (bind _squadmembers (Class|FortGameModeAthena|GetSquadMembers (Variables|Default|GetGameModeAthena) (Class|FortPlayerStateAthena|GetSquadId _converterplayerstateathena)))
      (bind _output_get (LocalVariables|SetFoundSquadMembers _squadmembers))
      (Utilities|Array|Shuffle _output_get)
      (for _array_element (LocalVariables|GetFoundSquadMembers)
        (if (<= (Math|Vector|Distance2D(Vector) (Transformation|GetActorLocation (PlayerState|GetPawn _array_element)) (Transformation|GetActorLocation (Variables|Default|GetPawn))) (ScalableFloat|GetValueatLevel (Variables|Default|GetDistanceToTeamMember)))
          (bind _asfort_player_pawn (Utilities|Casting|CastToFortPlayerPawn (PlayerState|GetPawn _array_element)))
          (Variables|Default|SetClosestTeamMemberFound _asfort_player_pawn)
          (return true)))
      (Variables|Default|SetClosestTeamMemberFound 0)
      (return false))
    (:Is%20Not%20Valid
      (return false))))
